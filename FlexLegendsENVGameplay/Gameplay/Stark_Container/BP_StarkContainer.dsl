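; Searchable Stark armory stand loot container with two mechanical arms, a screen and a glow ring, locked until the armory activates.
; A player entering sets a RepNotify flag that runs a one-time awake sequence, enabling interaction and random idle breaks.
; Searching plays arm item-give, shutdown effects and a local music stinger; looting updates the owning spawner's map icon.
; Animations are material-driven WPO, restarted by writing game time into scalar parameters.
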(event EventBeginPlay
  (|CallParentFunction)
  (Variables|Interaction|SetbAllowInteract false))

(event OnArmoryPlayerEntered(BP_BarnTowelArmory_ReactiveProp_Component)
  (params (FortPawn :type Fort%20Pawn%20Object%20Reference))
  (Variables|Default|SetArmory%20Door%20Opened%20and%20Activated true))

(event OnLinkedActorDestroyed(FortLinkToActor)
  (params (DamageCauser :type Actor%20Object%20Reference))
  ;[1 1 1] This will link this actor to what's immediately below it when it's not a floor. When it's link is destroyed, destroy the chest. (May not be necessary, but including for now.)
  (Actor|DestroyActor))

(event OnArmoryOpened(BP_BarnTowelArmory_ReactiveProp_Component))

(event Custom|DebugStarkStandAwake
  (Variables|Default|SetArmory%20Door%20Opened%20and%20Activated true))

(event EventOnSetSearched
  (bind _returnvalue (GameInstanceSubsystems|GetMusicEventSubsystem))
  ;[1 0.8054 0.6827] On Searched
  (Utilities|FlowControl|Sequence
    (:then_0
      (Networking|CanExecuteCosmeticEvents
        (:True
          (CallFunction|BlackScreenAnimation)
          (Audio|Components|Audio|Stop (Variables|Default|GetChest_Ambient_Sound))
          (CallFunction|TurnOffScreen)
          (CallFunction|StopAudiovisualIndicator)
          ;[0.2271 0.1964 0.7826 0.5] Music Stinger Logic
          (Networking|SwitchHasAuthority
            (:Remote
              (bind _returnvalue_2 (Container|GetSearchingPawn))
              (if (Pawn|IsLocallyViewed _returnvalue_2)
                (bind _returnvalue_3 (Audio|SpawnSound2D (Variables|Sound|GetMusicStinger)))
                (Audio|AddEventbyComponent _returnvalue _returnvalue_3 "(TagName=\"MusicEventSubsystem.Event.Stinger\")" "(GameplayTags=)" 0 "")
                (Utilities|FlowControl|Delay 5.0)
                (Audio|RemoveEvent _returnvalue 0)))))))
    (:then_1
      (Utilities|Time|ClearandInvalidateTimerbyHandle (Variables|Default|GetIdleBreak2))
      (Utilities|Time|ClearandInvalidateTimerbyHandle (Variables|Default|GetIdleBreak))
      (CallFunction|ArmsGiveItemWPO (Variables|Default|GetWPO_MechanicalArm_1) (Variables|Default|GetStarkItem))
      (Audio|SpawnSoundAttached :Sound (Variables|Sound|GetGiveItemSoundArm01) :AttachToComponent (Variables|Default|GetWPO_MechanicalArm_1) :AttachPointName "None" :Location "0.000000,0.000000,100.000000" :LocationType "KeepRelativeOffset" :bStopWhenAttachedToDestroyed false :VolumeMultiplier 1.0 :PitchMultiplier 1.0 :StartTime 0.0 :AttenuationSettings "/Game/Athena/Sounds/Props/Abduction_Pipes/Abduction_Pipe_Loop_Att.Abduction_Pipe_Loop_Att" :bAutoDestroy true)
      (Utilities|FlowControl|Delay (Variables|Default|GetAnimationDelay))
      (CallFunction|ArmsGiveItemWPO (Variables|Default|GetWPO_MechanicalArm_2) (Variables|Default|GetStarkItem1))
      (Audio|SpawnSoundAttached :Sound (Variables|Sound|GetGiveItemSoundArm02) :AttachToComponent (Variables|Default|GetWPO_MechanicalArm_1) :AttachPointName "None" :Location "0.000000,0.000000,100.000000" :LocationType "KeepRelativeOffset" :bStopWhenAttachedToDestroyed false :VolumeMultiplier 1.0 :PitchMultiplier 1.0 :StartTime 0.0 :AttenuationSettings "/Game/Athena/Sounds/Props/Abduction_Pipes/Abduction_Pipe_Loop_Att.Abduction_Pipe_Loop_Att" :bAutoDestroy true)
      (Utilities|Time|SetTimerbyEvent (delegate Custom|DesactivateArm) 1.33)
      (Rendering|SetVisibility (Variables|Default|GetStarkItem1) false false))))

(event Custom|DesactivateArm
  (bind _wpo_mechanicalarm_1 (Variables|Default|GetWPO_MechanicalArm_1))
  (bind _wpo_mechanicalarm_2 (Variables|Default|GetWPO_MechanicalArm_2))
  ;[1 0.8054 0.6827] On Searched
  (Rendering|Material|SetScalarParameterValueonMaterials _wpo_mechanicalarm_1 "AnimationIndex" 2.0)
  (Audio|SpawnSoundAttached (Variables|Sound|GetDeactivateSoundArm01) _wpo_mechanicalarm_1 "None" "0.000000,0.000000,100.000000")
  (Rendering|Material|SetScalarParameterValueonMaterials (Variables|Default|GetStarkItem) "BP_GetTime" (Utilities|Time|GetGameTimeinSeconds))
  (Utilities|FlowControl|Delay (Variables|Default|GetAnimationDelay))
  (Rendering|Material|SetScalarParameterValueonMaterials _wpo_mechanicalarm_2 "AnimationIndex" 2.0)
  (Audio|SpawnSoundAttached (Variables|Sound|GetDeactivateSoundArm02) _wpo_mechanicalarm_2 "None" "0.000000,0.000000,100.000000")
  (Rendering|Material|SetScalarParameterValueonMaterials (Variables|Default|GetStarkItem1) "BP_GetTime" (Utilities|Time|GetGameTimeinSeconds))
  (CallFunction|GlowingRingAnimation 0.0 1.5 0.9 0.0))

(event EventOnLoot
  ;[1 0.8054 0.6827] On Searched
  ;;[0.08633 0 0.15] OnLoot
  (Networking|SwitchHasAuthority
    (:Authority
      (bind _asb_armory_stand_spawner (Utilities|Casting|CastToB_ArmoryStand_Spawner (Actor|GetOwner)))
      (Class|BArmoryStandSpawner|SetMapIcontoLooted _asb_armory_stand_spawner) ; since the Spawner is always relevant, \nwe are using that to handle the map icon \nso that it can be seen across the BR map
    )))

(event Custom|AwakeSequence
  (bind _wpo_mechanicalarm_1 (Variables|Default|GetWPO_MechanicalArm_1))
  (bind _wpo_mechanicalarm_2 (Variables|Default|GetWPO_MechanicalArm_2))
  ;[1 1 1] Stand Activation
  (Utilities|FlowControl|DoOnce false
    (:Completed
      (Utilities|FlowControl|Sequence
        (:then_0
          (CallFunction|GlowingRingAnimation 0.0 1.5 0.9)
          (Rendering|Material|SetScalarParameterValueonMaterials _wpo_mechanicalarm_1 "BP_GetTime" (Utilities|Time|GetGameTimeinSeconds))
          (Audio|SpawnSoundAttached (Variables|Sound|GetAwakeSoundArm01) _wpo_mechanicalarm_1)
          (Utilities|FlowControl|Delay (Variables|Default|GetAnimationDelay))
          (Rendering|Material|SetScalarParameterValueonMaterials _wpo_mechanicalarm_2 "BP_GetTime" (Utilities|Time|GetGameTimeinSeconds))
          (Audio|SpawnSoundAttached (Variables|Sound|GetAwakeSoundArm02) _wpo_mechanicalarm_2 "None" "0.000000,0.000000,100.000000")
          (Utilities|Time|SetTimerbyEvent (delegate Custom|IdleAnimation) 1.33)
          (Variables|Interaction|SetbAllowInteract true))
        (:then_1
          (Networking|CanExecuteCosmeticEvents
            (:True
              (Rendering|Material|SetScalarParameterValueonMaterials (Variables|Default|GetMechanicalArm_Screen) "BP_GetTime" (Utilities|Time|GetGameTimeinSeconds)))))
        (:then_2
          (Networking|CanExecuteCosmeticEvents
            (:True
              (bind (_direction _location) (Timeline|BlueGlow PlayFromStart)
                (:Update
                  (Transformation|SetRelativeTransform (Variables|Default|GetBlueGlowRing) (Math|Transform|MakeTransform _location) true false))))))
        (:then_3
          (Variables|Container|SetLootSpawnLocation (Variables|/Script/Engine.SceneComponent|GetRelativeLocation (Variables|Default|GetStarkItem)))
          (Development|PrintString "loca is set"))))))

(event Custom|BlackScreenAnimation
  (bind _mechanicalarm_screen (Variables|Default|GetMechanicalArm_Screen))
  ;[1 1 1] Stand Activation
  (Rendering|Material|SetScalarParameterValueonMaterials _mechanicalarm_screen "InvertScreenAnimation" 1.0)
  (Rendering|Material|SetScalarParameterValueonMaterials _mechanicalarm_screen "BP_GetTime" (Utilities|Time|GetGameTimeinSeconds)))

(event Custom|IdleAnimation
  ;[1 1 1] Stand Activation
  (bind _returnvalue (Utilities|Time|SetTimerbyEvent (delegate Custom|IdleBreakArm1) (Math|Random|RandomFloatinRange 5.0 7.0) true))
  (Variables|Default|SetIdleBreak _returnvalue)
  (bind _returnvalue_2 (Utilities|Time|SetTimerbyEvent (delegate Custom|IdleBreakArm2) (Math|Random|RandomFloatinRange 2.0 6.0) true))
  (Variables|Default|SetIdleBreak2 _returnvalue_2))

(event Custom|IdleBreakArm1
  ;[1 1 1] Stand Activation
  (CallFunction|IdleBreakLogic (Variables|Default|GetWPO_MechanicalArm_2) (Variables|Default|GetStarkItem1)))

(event Custom|IdleBreakArm2
  ;[1 1 1] Stand Activation
  (CallFunction|IdleBreakLogic (Variables|Default|GetWPO_MechanicalArm_1) (Variables|Default|GetStarkItem)))

(event Custom|TurnOffScreen
  (bind _audio (Variables|Default|GetAudio))
  ;[1 1 1] Stand Activation
  (Utilities|FlowControl|Sequence
    (:then_0
      (Timeline|BlueGlow Reverse))
    (:then_1
      (Audio|Components|Audio|FadeOut _audio 0.2 0.0 "Linear" :!class /Script/Engine.AudioComponent)
      (Components|DestroyComponent _audio)
      (Rendering|SetVisibility (Variables|Default|GetBlueGlowRing) false false))))

(event Custom|GlowingRingAnimation
  (params (MinBrightness :type Float%20%28double-precision%29) (MaxBrightness :type Float%20%28double-precision%29 :default 1.0) (AnimationSpeed :type Float%20%28double-precision%29 :default 1.0) (TurnOn/Off-Index :type Float%20%28double-precision%29 :default 1.0))
  (bind _platform (Variables|Default|GetPlatform))
  ; GlowingRingAnimation
  (Utilities|FlowControl|Sequence
    (:then_0
      (Rendering|Material|SetScalarParameterValueonMaterials _platform "BP_GetTime" (Utilities|Time|GetGameTimeinSeconds))
      (Rendering|Material|SetScalarParameterValueonMaterials _platform "EmissiveBrightness_Min" MinBrightness)
      (Rendering|Material|SetScalarParameterValueonMaterials _platform "EmissiveBrightness_Max" MaxBrightness)
      (Rendering|Material|SetScalarParameterValueonMaterials _platform "Emissive_Speed" AnimationSpeed)
      (Rendering|Material|SetScalarParameterValueonMaterials _platform "Emissive_TurnOn/Off_Index" TurnOn/Off-Index))
    (:then_1
      (Networking|CanExecuteCosmeticEvents
        (:True
          (if (== TurnOn/Off-Index 1.0)
            (Audio|SpawnSoundatLocation (Variables|Sound|GetHologramOnSound) (Transformation|GetWorldLocation (Variables|Default|GetTreasureLight)))))))))

(event Custom|IdleBreakLogic
  (params (StaticMeshArm :type Static%20Mesh%20Component%20Object%20Reference) (StaticMeshItem :type Static%20Mesh%20Component%20Object%20Reference))
  (Rendering|Material|SetScalarParameterValueonMaterials StaticMeshArm "BP_GetTime" (Utilities|Time|GetGameTimeinSeconds))
  (Rendering|Material|SetScalarParameterValueonMaterials StaticMeshArm "ActivateIdleBreak" 1.0))

(event Custom|Arms_GiveItemWPO
  (params (StaticMeshArm :type Static%20Mesh%20Component%20Object%20Reference) (StaticMeshItem :type Static%20Mesh%20Component%20Object%20Reference))
  (Rendering|Material|SetScalarParameterValueonMaterials StaticMeshArm "ActivateIdleBreak" 0.0)
  (Rendering|Material|SetScalarParameterValueonMaterials StaticMeshArm "AnimationIndex" 1.0)
  (Rendering|Material|SetScalarParameterValueonMaterials StaticMeshArm "BP_GetTime" (Utilities|Time|GetGameTimeinSeconds)))

(fn BlueprintCanInteract
  (params (InteractingPawn :type Fort%20Pawn%20Object%20Reference) (InteractionBeingAttempted :type EInteractionBeingAttempted%20Enum :default "FirstInteraction") (InteractionType :type TInteractionType%20Enum :default "IT_NoInteraction"))
  (return (and (not (GameplayTags|HasAnyMatchingGameplayTags InteractingPawn (GameplayTags|MakeLiteralGameplayTagContainer "(GameplayTags=((TagName=\"Granted.LodeSwan\")))"))) (|CallParentFunction 0))))

(fn OnRep_ArmoryDoorOpenedandActivated
  (if (Variables|Default|GetArmory%20Door%20Opened%20and%20Activated)
    (CallFunction|AwakeSequence)))

(fn ConstructionScript
  (|CallParentFunction))
