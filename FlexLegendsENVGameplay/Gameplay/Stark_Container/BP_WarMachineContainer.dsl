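; War Machine armory loot container with three mechanical arms, locked until a player enters the armory.
; A RepNotify flag triggers the staggered wake-up, enabling interaction and looping idle breaks, though pawns tagged Granted.LodeSwan stay blocked.
; Searching makes the arms present items before shutting down, animated through timestamped WPO material parameters rather than skeletal animation.
; On loot, the server marks the map icon looted through the always-relevant owning spawner.
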(event EventBeginPlay
  (|CallParentFunction)
  (Variables|Interaction|SetbAllowInteract false))

(event OnArmoryPlayerEntered(BP_BarnTowelArmory_ReactiveProp_Component)
  (params (FortPawn :type Fort%20Pawn%20Object%20Reference))
  (Variables|Default|SetArmory%20Door%20Opened%20and%20Activated true))

(event OnArmoryOpened(BP_BarnTowelArmory_ReactiveProp_Component))

(event Custom|DebugWarMachineStandAwake
  (Variables|Default|SetArmory%20Door%20Opened%20and%20Activated true))

(event EventOnSetSearched
  (bind _audio (Variables|Default|GetAudio))
  ;[1 0.8054 0.6827] On Searched
  (Utilities|FlowControl|Sequence
    (:then_0
      (Networking|CanExecuteCosmeticEvents
        (:True
          (CallFunction|TurnoffScreen)
          (Audio|Components|Audio|Stop (Variables|Default|GetChest_Ambient_Sound))
          (CallFunction|TurnoffGlow)
          (CallFunction|StopAudiovisualIndicator)
          (Utilities|FlowControl|Delay 0.2)
          (Rendering|SetVisibility (Variables|Default|GetRedGlowRing) false false)
          (Networking|SwitchHasAuthority
            (:Remote
              (bind _returnvalue (Container|GetSearchingPawn))
              (if (Pawn|IsLocallyViewed _returnvalue)
                (Audio|SpawnSoundatLocation (Variables|Sound|GetChestOpenSound) (Transformation|GetWorldLocation (Variables|Default|GetAudio)))))))))
    (:then_1
      (Utilities|Time|ClearandInvalidateTimerbyHandle (Variables|Default|GetTimerIdleBreak1))
      (Utilities|Time|ClearandInvalidateTimerbyHandle (Variables|Default|GetTimerIdleBreak3))
      (Utilities|Time|ClearandInvalidateTimerbyHandle (Variables|Default|GetTimerIdleBreak2)))
    (:then_2
      (CallFunction|ArmsGiveItemWPO (Variables|Default|GetWPO_MechanicalArm_1) (Variables|Default|GetStarkItem2))
      (Audio|SpawnSoundAttached (Variables|Sound|GetGiveItemSoundArm01) (Variables|Default|GetWPO_MechanicalArm_1) "None" "0.000000,0.000000,100.000000")
      (Utilities|FlowControl|Delay (Variables|Default|GetAnimationDelay))
      (CallFunction|ArmsGiveItemWPO (Variables|Default|GetWPO_MechanicalArm_2) (Variables|Default|GetStarkItem))
      (Audio|SpawnSoundAttached (Variables|Sound|GetGiveItemSoundArm02) (Variables|Default|GetWPO_MechanicalArm_2) "None" "0.000000,0.000000,100.000000")
      (Utilities|FlowControl|Delay (Variables|Default|GetAnimationDelay))
      (CallFunction|ArmsGiveItemWPO (Variables|Default|GetWPO_MechanicalArm_3) (Variables|Default|GetStarkItem1))
      (Audio|SpawnSoundAttached (Variables|Sound|GetGiveItemSoundArm03) (Variables|Default|GetWPO_MechanicalArm_3))
      (Utilities|Time|SetTimerbyEvent (delegate Custom|Shutdown) 1.33)
      (Rendering|SetVisibility (Variables|Default|GetStarkItem2) false false)
      (Audio|SpawnSoundAttached :Sound "/Game/Sounds/Fort_Gadgets/Athena/MechanicalEngineerFemale/Glider_MechanicalEngineerFemale_Open_Base.Glider_MechanicalEngineerFemale_Open_Base" :AttachToComponent (Variables|Debug|GetStaticMeshComponent) :AttachPointName "None" :LocationType "KeepRelativeOffset" :bStopWhenAttachedToDestroyed false :VolumeMultiplier 1.0 :PitchMultiplier 1.0 :StartTime 0.0 :AttenuationSettings "/Game/Athena/Sounds/Props/Abduction_Pipes/Abduction_Pipe_Loop_Att.Abduction_Pipe_Loop_Att" :bAutoDestroy true)
      (Audio|Components|Audio|FadeOut _audio 0.2 0.0 "Linear" :!class /Script/Engine.AudioComponent)
      (Components|DestroyComponent _audio))))

(event Custom|IdleLoopLogic
  (bind _returnvalue (Math|Random|RandomFloatinRange 15.0 2e+01))
  (bind _returnvalue_2 (Utilities|Time|SetTimerbyEvent (delegate Custom|IdleBreak1) _returnvalue true))
  (Variables|Default|SetTimerIdleBreak1 _returnvalue_2)
  (bind _returnvalue_3 (Utilities|Time|SetTimerbyEvent (delegate Custom|IdleBreak2) _returnvalue true))
  (Variables|Default|SetTimerIdleBreak2 _returnvalue_3)
  (bind _returnvalue_4 (Utilities|Time|SetTimerbyEvent (delegate Custom|IdleBreak3) _returnvalue true))
  (Variables|Default|SetTimerIdleBreak3 _returnvalue_4))

(event Custom|Shutdown
  (bind _wpo_mechanicalarm_1 (Variables|Default|GetWPO_MechanicalArm_1))
  ;[1 0.8054 0.6827] On Searched
  ;[0.15 0.01798 0.02658 0.5] Deactivate
  (Rendering|Material|SetScalarParameterValueonMaterials _wpo_mechanicalarm_1 "AnimationIndex" 2.0)
  (Utilities|FlowControl|Sequence
    (:then_0
      (Rendering|Material|SetScalarParameterValueonMaterials _wpo_mechanicalarm_1 "BP_GetTime" (Utilities|Time|GetGameTimeinSeconds))
      (Audio|SpawnSoundAttached (Variables|Sound|GetDactivateSoundArm03) (Variables|Default|GetWPO_MechanicalArm_3))
      (CallFunction|GlowingRingAnimation 0.0 1.5 0.9 0.0))
    (:then_1
      (Audio|SpawnSoundAttached (Variables|Sound|GetDactivateSoundArm01) (Variables|Default|GetWPO_MechanicalArm_1) "None" "0.000000,0.000000,100.000000")
      (Audio|SpawnSoundAttached (Variables|Sound|GetDactivateSoundArm02) (Variables|Default|GetWPO_MechanicalArm_2) "None" "0.000000,0.000000,100.000000"))))

(event Custom|IdleBreak1
  (CallFunction|IdleBreakLogic (Variables|Default|GetWPO_MechanicalArm_1) (Variables|Default|GetStarkItem2)))

(event Custom|IdleBreak2
  (CallFunction|IdleBreakLogic (Variables|Default|GetWPO_MechanicalArm_2) (Variables|Default|GetStarkItem)))

(event Custom|IdleBreak3
  (CallFunction|IdleBreakLogic (Variables|Default|GetWPO_MechanicalArm_3) (Variables|Default|GetStarkItem1)))

(event Custom|ActivateAwakeSequence
  ;[1 1 1] give
  (Utilities|FlowControl|DoOnce false
    (:Completed
      (Utilities|FlowControl|Sequence
        (:then_0
          (Rendering|Material|SetScalarParameterValueonMaterials (Variables|Default|GetWPO_MechanicalArm_1) "BP_GetTime" (Utilities|Time|GetGameTimeinSeconds))
          (Audio|SpawnSoundAttached (Variables|Sound|GetAwakeSoundArm01) (Variables|Default|GetWPO_MechanicalArm_1) "None" "0.000000,0.000000,100.000000")
          (Utilities|FlowControl|Delay (Variables|Default|GetAnimationDelay))
          (Rendering|Material|SetScalarParameterValueonMaterials (Variables|Default|GetWPO_MechanicalArm_2) "BP_GetTime" (Utilities|Time|GetGameTimeinSeconds))
          (Audio|SpawnSoundAttached (Variables|Sound|GetAwakeSoundArm02) (Variables|Default|GetWPO_MechanicalArm_2) "None" "0.000000,0.000000,100.000000")
          (Utilities|FlowControl|Delay (Variables|Default|GetAnimationDelay))
          (Rendering|Material|SetScalarParameterValueonMaterials (Variables|Default|GetWPO_MechanicalArm_3) "BP_GetTime" (Utilities|Time|GetGameTimeinSeconds))
          (Audio|SpawnSoundAttached (Variables|Sound|GetAwakeSoundArm03) (Variables|Default|GetWPO_MechanicalArm_3) "None" "0.000000,0.000000,100.000000")
          (Utilities|Time|SetTimerbyEvent (delegate Custom|IdleAnimation) 1.33)
          (Variables|Interaction|SetbAllowInteract true))
        (:then_1
          (Networking|CanExecuteCosmeticEvents
            (:True
              (Rendering|Material|SetScalarParameterValueonMaterials (Variables|Default|GetMechanicalArm_Screen) "BP_GetTime" (Utilities|Time|GetGameTimeinSeconds))
              (CallFunction|GlowingRingAnimation 0.0 1.5 0.9))))
        (:then_2
          (Networking|CanExecuteCosmeticEvents
            (:True
              (bind (_direction _location) (Timeline|Timeline_1 PlayFromStart)
                (:Update
                  (Transformation|SetRelativeTransform (Variables|Default|GetRedGlowRing) (Math|Transform|MakeTransform _location) true false))))))))))

(event Custom|TurnoffScreen
  (bind _mechanicalarm_screen (Variables|Default|GetMechanicalArm_Screen))
  ;[1 1 1] give
  (Rendering|Material|SetScalarParameterValueonMaterials _mechanicalarm_screen "InvertScreenAnimation" 1.0)
  (Rendering|Material|SetScalarParameterValueonMaterials _mechanicalarm_screen "BP_GetTime" (Utilities|Time|GetGameTimeinSeconds))
  (CallFunction|GlowingRingAnimation 0.0 1.5 0.9))

(event Custom|IdleAnimation
  ;[1 1 1] give
  (CallFunction|IdleLoopLogic))

(event EventOnLoot
  ;[0.08633 0 0.15] OnLoot
  (Networking|SwitchHasAuthority
    (:Authority
      (bind _asb_armory_stand_spawner (Utilities|Casting|CastToB_ArmoryStand_Spawner (Actor|GetOwner)))
      (Class|BArmoryStandSpawner|SetMapIcontoLooted _asb_armory_stand_spawner) ; since the Spawner is always relevant, \nwe are using that to handle the map icon \nso that it can be seen across the BR map
    )))

(event Custom|TurnoffGlow
  ;[1 1 1] give
  (Timeline|Timeline_1 Reverse))

(event Custom|GlowingRingAnimation
  (params (MinBrightness :type Float%20%28double-precision%29) (MaxBrightness :type Float%20%28double-precision%29 :default 1.0) (AnimationSpeed :type Float%20%28double-precision%29 :default 1.0) (TurnOn/Off-Index :type Float%20%28double-precision%29 :default 1.0))
  (bind _platform (Variables|Default|GetPlatform))
  ; GlowingRingAnimation
  (Utilities|FlowControl|Sequence
    (:then_0
      (Rendering|Material|SetScalarParameterValueonMaterials _platform "BP_GetTime" (Utilities|Time|GetGameTimeinSeconds))
      (Rendering|Material|SetScalarParameterValueonMaterials _platform "EmissiveBrightness_Min" MinBrightness)
      (Rendering|Material|SetScalarParameterValueonMaterials _platform "EmissiveBrightness_Max" MaxBrightness)
      (Rendering|Material|SetScalarParameterValueonMaterials _platform "Emissive_Speed" AnimationSpeed)
      (Rendering|Material|SetScalarParameterValueonMaterials _platform "Emissive_TurnOn/Off_Index" TurnOn/Off-Index))
    (:then_1
      (Networking|CanExecuteCosmeticEvents
        (:True
          (if (== TurnOn/Off-Index 1.0)
            (Audio|SpawnSoundatLocation (Variables|Sound|GetHologramOnSound) (Transformation|GetWorldLocation (Variables|Default|GetTreasureLight)))))))))

(event Custom|IdleBreakLogic
  (params (StaticMeshArm :type Static%20Mesh%20Component%20Object%20Reference) (StaticMeshItem :type Static%20Mesh%20Component%20Object%20Reference))
  (Rendering|Material|SetScalarParameterValueonMaterials StaticMeshArm "BP_GetTime" (Utilities|Time|GetGameTimeinSeconds))
  (Rendering|Material|SetScalarParameterValueonMaterials StaticMeshArm "ActivateIdleBreak" 1.0))

(event Custom|Arms_GiveItemWPO
  (params (StaticMeshArm :type Static%20Mesh%20Component%20Object%20Reference) (StaticMeshItem :type Static%20Mesh%20Component%20Object%20Reference))
  (Rendering|Material|SetScalarParameterValueonMaterials StaticMeshArm "ActivateIdleBreak" 0.0)
  (Rendering|Material|SetScalarParameterValueonMaterials StaticMeshArm "AnimationIndex" 1.0)
  (Rendering|Material|SetScalarParameterValueonMaterials StaticMeshArm "BP_GetTime" (Utilities|Time|GetGameTimeinSeconds)))

(fn BlueprintCanInteract
  (params (InteractingPawn :type Fort%20Pawn%20Object%20Reference) (InteractionBeingAttempted :type EInteractionBeingAttempted%20Enum :default "FirstInteraction") (InteractionType :type TInteractionType%20Enum :default "IT_NoInteraction"))
  (return (and (not (GameplayTags|HasAnyMatchingGameplayTags InteractingPawn (GameplayTags|MakeLiteralGameplayTagContainer "(GameplayTags=((TagName=\"Granted.LodeSwan\")))"))) (|CallParentFunction 0))))

(fn OnRep_ArmoryDoorOpenedandActivated
  (if (Variables|Default|GetArmory%20Door%20Opened%20and%20Activated)
    (CallFunction|ActivateAwakeSequence)))

(fn ConstructionScript
  (|CallParentFunction))
