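; Spawner for armory stand containers that picks a Stark or War Machine Armory class and drives its map icon.
; On landing it sets the class to spawn by variant; on spawning it stores the variant and activates its icon component.
; Spawned actors are registered under a unique location tag to receive armory state events and owned by the spawner, letting the container call back to switch the icon to looted.
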(event EventBeginPlay)

(event OnArmoryLanded(BP_BarnTowelArmory_ReactiveProp_Component)
  (params (SelectedVariant :type Enum_BarnTowelArmory_SelectedVariant%20Enum :default "NewEnumerator0"))
  (switch Utilities|FlowControl|Switch|SwitchonEnum_BarnTowelArmory_SelectedVariant SelectedVariant
    (:NewEnumerator0
      _)
    (:NewEnumerator2
      _)
    (:NewEnumerator1
      (Variables|SpawnerVars|SetClassToSpawn (Variables|Default|GetWar%20Machine%20Armory))
      (|CallParentFunction))
    (:NewEnumerator3
      _))
  (Variables|SpawnerVars|SetClassToSpawn (Variables|Default|GetStark%20Armory)))

(event OnArmorySpawned(BP_BarnTowelArmory_ReactiveProp_Component)
  (params (SpawnedVariant :type Enum_BarnTowelArmory_SelectedVariant%20Enum :default "NewEnumerator0"))
  ;[0.08633 0 0.15] OnArmorySpawned
  ; Set and activate the map icon to the correct variant
  (bind _output_get (Variables|Default|SetSelectedArmoryVariant SpawnedVariant))
  (switch Utilities|FlowControl|Switch|SwitchonEnum_BarnTowelArmory_SelectedVariant _output_get
    (:NewEnumerator0
      _)
    (:NewEnumerator2
      _)
    (:NewEnumerator1
      (Components|Activation|Activate (Variables|Default|GetAthenaSpecialActor_MurkyVoleActive) false))
    (:NewEnumerator3
      _))
  (Components|Activation|Activate (Variables|Default|GetAthenaSpecialActor_BarnTowelActive) false))

(event EventServerActorSpawned
  (params (Actor :type Actor%20Object%20Reference))
  (Class|FortGlobalActorRegistrySubsystem|RegisterActor (WorldSubsystems|GetFortGlobalActorRegistrySubsystem) (Variables|Default|GetUniqueLocationTag) Actor) ; register the Container to receive Armory state events
  (Actor|SetOwner Actor self) ; set this actor as the owner so that \nthe Container has a reference back to this actor
)

(event Custom|SetMapIconToLooted
  ;[0.08633 0 0.15] SetMapIconToLooted
  ; Set the map icon to the looted state - event is called from the Spawned container
  (switch Utilities|FlowControl|Switch|SwitchonEnum_BarnTowelArmory_SelectedVariant (Variables|Default|GetSelectedArmoryVariant)
    (:NewEnumerator0
      _)
    (:NewEnumerator2
      _)
    (:NewEnumerator1
      (Components|Activation|Activate (Variables|Default|GetAthenaSpecialActor_MurkyVoleClaimed) false)
      (Components|Activation|Deactivate (Variables|Default|GetAthenaSpecialActor_MurkyVoleActive)))
    (:NewEnumerator3
      _))
  (Components|Activation|Activate (Variables|Default|GetAthenaSpecialActor_BarnTowelClaimed) false)
  (Components|Activation|Deactivate (Variables|Default|GetAthenaSpecialActor_BarnTowelActive)))

(fn ConstructionScript
  (|CallParentFunction))
