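# Default Lit glow material for the Stark Container gameplay prop, judging by its name and folder.
# A constant color is multiplied by the EmissionStrength scalar parameter, so material instances can tune intensity but not hue.
# Despite the GreenGlow name, the constant is almost pure red.
# The graph has no textures or other logic, and the listing does not show which material output the product feeds.
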Material /FlexLegendsENVGameplay/Gameplay/Stark_Container/M_GreenGlow_StarkContainer
graph {
 properties {bAutomaticallySetUsageInEditor=false, pixelDepthOffsetMode=PDOM_Legacy, usedShadingModels=Default Lit}
 nodes [Multiply(A=Constant3Vector() {constant=<rgba 1 0 0.0168 1>}, B=ScalarParameter() {parameterName=EmissionStrength})]
}
