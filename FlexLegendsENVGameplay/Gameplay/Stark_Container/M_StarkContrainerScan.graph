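# Default Lit material for a scanner element on the Stark Container, judging by its name and texture.
# It multiplies the RGB of the MilitaryBase scanner diffuse texture by a TextureAppear scalar parameter, which can scale the image's intensity from material instances.
# The graph shows no connection from that product to a material output.
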Material /FlexLegendsENVGameplay/Gameplay/Stark_Container/M_StarkContrainerScan
graph {
 properties {bAutomaticallySetUsageInEditor=false, pixelDepthOffsetMode=PDOM_Legacy, usedShadingModels=Default Lit}
 nodes [Multiply(
   A=TextureSample() {texture=/Game/Environments/Sets/MilitaryBase/Textures/T_MilitaryBase_scanner_D}.RGB
   B=ScalarParameter() {parameterName=TextureAppear}
  )]
}
